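# Animation Blueprint for a female Commando ultra-rare head, built to add secondary motion to its hair bones, though the output currently only passes the input pose through.
# A chain of six Anim Dynamics nodes simulates the dyn_hair bones with tight angular limits, weighted by pawn 2D speed, but is not wired to the output.
# Copy Pose From Mesh and RigidBody nodes also sit unconnected, while initialization still stores the owner's skeletal mesh in MeshToCopy.
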AnimBlueprint /BRCosmetics/Characters/Player/Female/Medium/Heads/F_MED_UltraRareCommando_01/Meshes/F_MED_BLK_Commando_UltraRare_Head_01_Dyn_AnimBP
target_skeleton=/Game/Characters/Player/Common/Fortnite_Base_Head/Fortnite_Base_Head_Export_Skeleton
graphs [
 AnimGraph {nodes [
   n1 = MapRangeClamped(Value=Get PawnSpeed2D(), InRangeA=0, InRangeB=100, OutRangeA=.2)
   n2 = float * float(A=@n1, B=1)
   n3 = Anim Dynamics - Bone: dyn_hair_01(Alpha=@n2) {
    angularBiasOverride=1
    bGravityOverrideInSimSpace=true
    bOverrideAngularBias=true
    gravityScale=0.14
    alphaBoolBlend {blendOption=Linear}
    boundBone {boneName=dyn_hair_01}
    physicsBodyDefinitions [{
      sphereCollisionRadius=0
      boundBone {boneName=dyn_hair_01}
      boxExtents=<5 5 5>
      constraintSetup {angularLimitsMax=<10 0 15>, angularLimitsMin=<-10 0 -15>, angularTarget=<0 0 0>, linearAxesMax=<0 0 2>}
      localJointOffset=<0 0 -2>
     }]
   }
   n4 = Anim Dynamics - Bone: dyn_hair_02(ComponentPose=@n3, Alpha=@n2) {
    angularBiasOverride=1
    bGravityOverrideInSimSpace=true
    bOverrideAngularBias=true
    gravityScale=0.1
    alphaBoolBlend {blendOption=Linear}
    boundBone {boneName=dyn_hair_02}
    physicsBodyDefinitions [{
      sphereCollisionRadius=0
      boundBone {boneName=dyn_hair_02}
      boxExtents=<5 5 5>
      constraintSetup {
       angularTargetAxis=AxisY
       twistAxis=AxisY
       angularLimitsMax=<10 0 10>
       angularLimitsMin=<-10 0 -10>
       angularTarget=<0 0 0>
       linearAxesMax=<0 0 3>
      }
      localJointOffset=<0 0 -2>
     }]
    relativeSpaceBone {boneName=dyn_hair_02}
   }
   n5 = Anim Dynamics - Bone: dyn_hair_03(ComponentPose=@n4, Alpha=@n2, ExternalForce="0,0,0") {
    angularBiasOverride=1
    bGravityOverrideInSimSpace=true
    bOverrideAngularBias=true
    gravityScale=0.2
    alphaBoolBlend {blendOption=Linear}
    boundBone {boneName=dyn_hair_03}
    physicsBodyDefinitions [{
      sphereCollisionRadius=0
      boundBone {boneName=dyn_hair_03}
      boxExtents=<5 5 5>
      constraintSetup {
       angularTargetAxis=AxisY
       twistAxis=AxisY
       angularLimitsMax=<10 0 10>
       angularLimitsMin=<0 0 -10>
       angularTarget=<0 0 0>
       linearAxesMax=<0 0 3>
      }
      localJointOffset=<0 0 -2>
     }]
    relativeSpaceBone {boneName=dyn_hair_03}
   }
   n6 = Anim Dynamics - Bone: dyn_hair_04(ComponentPose=@n5, Alpha=@n2, ExternalForce="0,0,0") {
    angularBiasOverride=1
    bGravityOverrideInSimSpace=true
    bOverrideAngularBias=true
    gravityScale=0.1
    alphaBoolBlend {blendOption=Linear}
    boundBone {boneName=dyn_hair_04}
    physicsBodyDefinitions [{
      sphereCollisionRadius=0
      boundBone {boneName=dyn_hair_04}
      boxExtents=<5 5 5>
      constraintSetup {
       angularTargetAxis=AxisY
       twistAxis=AxisY
       angularLimitsMax=<10 0 10>
       angularLimitsMin=<-10 0 -10>
       angularTarget=<0 0 0>
       linearAxesMax=<0 0 3>
      }
      localJointOffset=<0 0 -2>
     }]
    relativeSpaceBone {boneName=dyn_hair_04}
   }
   n7 = Anim Dynamics - Bone: dyn_hair_05(ComponentPose=@n6, Alpha=@n2, ExternalForce="0,0,0") {
    angularBiasOverride=1
    bGravityOverrideInSimSpace=true
    bOverrideAngularBias=true
    gravityScale=0.1
    alphaBoolBlend {blendOption=Linear}
    boundBone {boneName=dyn_hair_05}
    physicsBodyDefinitions [{
      sphereCollisionRadius=0
      boundBone {boneName=dyn_hair_05}
      boxExtents=<5 5 5>
      constraintSetup {
       angularTargetAxis=AxisY
       twistAxis=AxisY
       angularLimitsMax=<10 0 10>
       angularLimitsMin=<-10 0 -10>
       angularTarget=<0 0 0>
       linearAxesMax=<0 0 3>
      }
      localJointOffset=<0 0 -2>
     }]
   }
   Anim Dynamics - Bone: dyn_hair_06(ComponentPose=@n7, Alpha=float * float(A=@n1, B=.5), ExternalForce="0,0,0") {
    angularBiasOverride=1
    bGravityOverrideInSimSpace=true
    bOverrideAngularBias=true
    gravityScale=0.13
    alphaBoolBlend {blendOption=Linear}
    boundBone {boneName=dyn_hair_06}
    physicsBodyDefinitions [{
      sphereCollisionRadius=0
      boundBone {boneName=dyn_hair_06}
      boxExtents=<5 5 5>
      constraintSetup {
       angularTargetAxis=AxisY
       twistAxis=AxisY
       angularLimitsMax=<3 0 10>
       angularLimitsMin=<-10 0 -10>
       angularTarget=<0 0 0>
       linearAxesMax=<0 0 1>
      }
      localJointOffset=<0 0 -2>
     }]
   }
   Copy Pose From Mesh(SourceMeshComponent=Get MeshToCopy() -> (MeshToCopy:Skeletal Mesh Component Object Reference))
   RigidBody() {alphaBoolBlend {blendOption=Linear}}
   AnimGraphNode_SubInput_0 = Input Pose()
   Output Pose(Result=Component To Local(ComponentPose=Local To Component(LocalPose=@AnimGraphNode_SubInput_0))) {layerGroup=None}
  ]}
 EventGraph {nodes [
   Event BlueprintInitializeAnimation()
   > Set MeshToCopy(MeshToCopy:Skeletal Mesh Component Object Reference=GetOwnerSkeletalMeshComponent()) -> (Output_Get:Skeletal Mesh Component Object Reference)
   TryGetPawnOwner()
   Event BlueprintUpdateAnimation()
  ]}
]
variables [MeshToCopy {type=SkeletalMeshComponent}]
